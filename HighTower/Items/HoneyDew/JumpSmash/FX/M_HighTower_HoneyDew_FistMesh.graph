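# Translucent unlit mesh-particle material for the fist mesh effect in the HighTower HoneyDew JumpSmash item.
# Color multiplies a panning swirl-noise gradient by particle color and by dark shading built from reused Sea Salt pickaxe textures, UV falloff and a normal-mapped Fresnel.
# Opacity combines particle alpha, view-angle fade, a UV edge falloff and noise erosion driven by the Erode dynamic parameter.
# Power-warped U coordinates stretch the noise, and a scalar blends between UV channels.
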Material /HighTower/Items/HoneyDew/JumpSmash/FX/M_HighTower_HoneyDew_FistMesh
graph {
 properties {bUsedWithMeshParticles=true, blendMode=BLEND_Translucent, pixelDepthOffsetMode=PDOM_Legacy, shadingModel=MSM_Unlit, usedShadingModels=Unlit}
 nodes [
  n1 = VectorParameter() {parameterName=ColorLow, defaultValue=<rgba 0.0257 1 0.533 1>}
  n2 = VectorParameter() {parameterName=ColorHigh, defaultValue=<rgba 13.4467 20 9.688 1>}
  n3 = MaterialFunctionCall(Float2=LinearInterpolate(
    A=TextureCoordinate()
    B=TextureCoordinate() {coordinateIndex=1, uTiling=0.8, vTiling=0.7}
    Alpha=ScalarParameter() {defaultValue=1, parameterName="TexCoord(0or1)"}
   )) {materialFunction=&BreakOutFloat2Components}
  n4 = Power(Base=@n3.R) {constExponent=0.3}
  n5 = TextureSample(UVs=Panner(Coordinate=AppendVector(A=@n4, B=@n3.G)) {speedX=-3, speedY=0.5}) {samplerType=SAMPLERTYPE_LinearGrayscale, texture=/Game/Global/Textures/Default/Noise/T_SwirlNoise_N_height}
  n6 = Power(Base=@n5.R)
  n7 = OneMinus(None=MaterialFunctionCall(Float2=TextureCoordinate() {coordinateIndex=1}) {materialFunction=&BreakOutFloat2Components}.R)
  n8 = Add(
   A=Power(Base=TextureSample() {texture=/Game/Weapons/FORT_Melee/Pickaxe_Sea_Salt/Textures/T_Pickaxe_Sea_Salt_D}.G) {constExponent=4}
   B=MaterialFunctionCall(In=@n5.R, Contrast=Constant() {r=0.7}) {materialFunction=&CheapContrast}.Result
  )
  n9 = LinearInterpolate(A=VectorParameter() {parameterName=ColorDark, defaultValue=<rgba 0.0101 0.0833 0.0057 1>}.RGB, Alpha=Saturate(None=Subtract(A=@n7, B=@n8)))
  n10 = Transform(None=TextureSample() {samplerType=SAMPLERTYPE_Normal, texture=/Game/Weapons/FORT_Melee/Pickaxe_Sea_Salt/Textures/T_Pickaxe_Sea_Salt_N}.RGB)
  n11 = Multiply(A=LinearInterpolate(
    A=VectorParameter() {parameterName=ColorDark, defaultValue=<rgba 0.0101 0.0833 0.0057 1>}.RGB
    Alpha=Fresnel(Normal=@n10) {baseReflectFraction=0, exponent=9}
   )) {constB=4}
  n12 = ParticleColor()
  Multiply(A=Multiply(A=LinearInterpolate(A=@n1.RGB, B=@n2.RGB, Alpha=@n6), B=Multiply(A=@n9, B=@n11)), B=@n12.RGB)
  n13 = Constant() {r=1}
  n14 = MaterialFunctionCall(Float2=LinearInterpolate(A=TextureCoordinate(), B=TextureCoordinate() {coordinateIndex=1}, Alpha=ScalarParameter() {defaultValue=1, parameterName="TexCoord(0or1)"})) {materialFunction=&BreakOutFloat2Components}
  n15 = OneMinus(None=@n14.R)
  n16 = Saturate(None=Power(Base=Subtract(A=@n15) {constB=0.13}))
  n17 = Power(Base=@n14.R) {constExponent=1}
  n18 = MaterialFunctionCall(Float2=LinearInterpolate(
    A=TextureCoordinate()
    B=TextureCoordinate() {coordinateIndex=1, vTiling=0.8}
    Alpha=ScalarParameter() {defaultValue=1, parameterName="TexCoord(0or1)"}
   )) {materialFunction=&BreakOutFloat2Components}
  n19 = Power(Base=@n18.R) {constExponent=0.5}
  n20 = TextureSample(UVs=Panner(Coordinate=AppendVector(A=@n19, B=@n18.G)) {speedX=-2}) {samplerType=SAMPLERTYPE_LinearGrayscale, texture=&T_SwirlNoise_N_height}
  n21 = Saturate(None=Subtract(A=DynamicParameter() {paramNames [Erode,Param2,Param3,Param4]}.Erode, B=Multiply(A=@n17, B=@n20.R)) {constA=0})
  Multiply(
   A=@n12.A
   B=Multiply(
    A=MaterialFunctionCall("FadeSensitivity (optional)"=@n13) {materialFunction=&FadeBasedOnViewAngle}.Result
    B=Multiply(A=@n16, B=@n21)
   )
  )
 ]
}
