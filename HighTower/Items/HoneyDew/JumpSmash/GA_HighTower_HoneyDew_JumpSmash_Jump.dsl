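; Gameplay ability for the jump of HoneyDew's JumpSmash, extending a parent charged-leap ability and cancelling building on activation.
; On launch, the server multicast-links the HoneyDew animation layer, and sufficiently charged jumps run targeting from above the player to break overhead objects.
; Montage notifies either send the jump-loop event or run parent logic and bind landing, so early landings stop the montage.
; The held-input cue receives the max charge time.
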(event EventOnAbilityActivated
  ;[1 1 1] Cancel building on activate
  (|CallParentFunction)
  (Building|ForceCancelBuildingTool (Fortnite|GetFortPlayerControllerfromActor (Ability|GetActivatingPawn))))

(event EventLaunchPlayer
  ;[1 1 1] When the player is launched, also target and break the things above them if they are doing a big jump.
  ;;[1 1 1] Linking Layer AnimBP
  (Utilities|IsServer
    (:true
      (Animation|MulticastLinkAnimLayer (Ability|GetActivatingPawn) "/HighTower/Animation/HoneyDew/HighTower_HoneyDew_LayerAnimBP.HighTower_HoneyDew_LayerAnimBP_C"))
    (:false
      _))
  (if (>= (Variables|PowerCharge|GetChargePercent) (ScalableFloat|GetValueatLevel (Variables|PowerTargetWhileLeaping|GetTargetWhileLeapChargeThreshold)))
    (CallFunction|Target)
    (else
      _))
  (|CallParentFunction))

(event EventCastMontageNotifyStart
  (params (NotifyTag :type Gameplay%20Tag%20Structure))
  (bind _event_jumploop (Variables|Default|GetEvent_JumpLoop))
  ;[1 1 1] If the notify is the looping jump send that event otherwise do the normal path.
  (if (GameplayTags|Equal(GameplayTag) NotifyTag _event_jumploop)
    (Ability|SendGameplayEvent _event_jumploop)
    (else
      (|CallParentFunction NotifyTag)
      (EventDispatchers|BindEventtoOnPawnLanded (Ability|GetActivatingPawn) (EventDispatchers|CreateEvent :!event Custom|OnLandEarly)))))

(event Ability|EventOnEndAbility
  (params (bWasCancelled :type Boolean))
  ;[1 1 1] If player lands early stop montage
  (|CallParentFunction bWasCancelled)
  (EventDispatchers|UnbindEventfromOnPawnLanded (Ability|GetActivatingPawn) (delegate Custom|OnLandEarly)))

(event Custom|OnLandEarly
  (params (Hit :type Hit%20Result%20Structure%20%28by%20ref%29))
  ;[1 1 1] If player lands early stop montage
  (Ability|Animation|MontageStop 0.1))

(fn GetCustomAbilitySourceTransform
  ;[1 1 1] Take the player's location and shift it up a little for the targeting location
  (return (Math|Transform|MakeTransform (+ (Transformation|GetActorLocation (Ability|GetAvatarActorfromActorInfo)) (Variables|Default|GetJumpTargetOffset)) (Transformation|GetActorRotation (Ability|GetAvatarActorfromActorInfo)))))

(fn InputHeldLoopGCParams
  (return (Ability|GameplayCue|MakeGameplayCueParameters 0.0 (ScalableFloat|GetValueatLevel (Variables|PowerCharge|GetRow_MaxChargeTime)))))
