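; Gameplay ability for the HighTower HoneyDew JumpSmash that plays the jump loop montage on the player pawn while it falls.
; Activation binds to movement mode changes, starts the montage and, on the server, multicasts an unlink of the HoneyDew anim layer.
; The ability ends when the movement mode is no longer Falling or the pawn cast fails.
; On end, it stops the loop montage only if that is still current.
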(event Custom|OnMovementModeChanged
  (params (Character :type Character%20Object%20Reference) (PrevMovementMode :type EMovementMode%20Enum :default "MOVE_None") (PreviousCustomMode :type Byte))
  ;[1 1 1] Bind to movement mode change. If it changes from falling end the ability
  (if (Utilities|Enum|NotEqual(Enum) (Variables|/Script/Engine.CharacterMovementComponent|GetMovementMode (Variables|/Script/Engine.Character|GetCharacterMovement Character)) "MOVE_Falling")
    (Ability|EndAbility)))

(event Ability|EventOnEndAbility
  (params (bWasCancelled :type Boolean))
  ;[1 1 1] On End ability stop the fall loop
  (CallFunction|StoptheFallLoop))

(event Ability|EventActivateAbility
  ;[1 1 1] When Triggered Start the anim Loop for the Jump, if we hit the ground or start another animation end
  (bind _asfort_player_pawn (Utilities|Casting|CastToFortPlayerPawn (Ability|GetAvatarActorfromActorInfo))
    (:then
      (Character|BindEventtoMovementModeChangedDelegate _asfort_player_pawn (EventDispatchers|CreateEvent :!event Custom|OnMovementModeChanged))
      (Ability|Tasks|PlayMontageAndWaitNotify "Cast Animation" (Variables|Default|GetJumpLoop)
        (:OnInterrupted
          _)
        (:OnCancelled
          _))
      (Utilities|IsServer
        (:true
          (Animation|MulticastUnlinkAnimLayer (Ability|GetActivatingPawn) "/HighTower/Animation/HoneyDew/HighTower_HoneyDew_LayerAnimBP.HighTower_HoneyDew_LayerAnimBP_C"))))
    (:CastFailed
      (Ability|EndAbility))))

(event Custom|StopTheFallLoop
  ;[1 1 1] Stop The Fall Loop - End the fall animation if it is playing
  (if (Utilities|Equal(Object) (Animation|GetCurrentMontage) (Variables|Default|GetJumpLoop))
    (Ability|Animation|MontageStop -1.0)))
