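; Gameplay ability for the knockback of HighTower's HoneyDew JumpSmash item that changes gravity once the launch finishes.
; It overrides the launch-finished event to apply a gravity gameplay effect to the owner, setting low gravity, then calls the parent implementation.
; An author note says zero gravity is probably wanted, but gravity appears to be clamped at present.
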(event EventLaunchFinished
  ;[1 1 1] Set Low Grav on finished.
  (Ability|ApplyGameplayEffectToOwner (Variables|Default|GetGE_Grav) 1 0) ; Note we probably want 0g but we seem to clamp the gravity at the moment.
  (|CallParentFunction))
