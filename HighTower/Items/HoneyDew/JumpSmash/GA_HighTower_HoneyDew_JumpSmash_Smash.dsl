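; HoneyDew jump smash ability: a leap that launches the player along their aim, with clamped pitch and data-driven speed.
; On leap, a looping timer forces the land logic once the character stops falling, clearing the infinite jump cooldown.
; The server multicasts anim layer link and unlink on leap and end.
; Landing adds a delayed SetBlockMovement failsafe and passes landed-in-water as the gameplay cue's normalized magnitude.
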(event EventLeap
  ;[1 1 1] Check to see if player is already on the ground
  (|CallParentFunction)
  (bind _returnvalue (Utilities|Time|SetTimerbyEvent (EventDispatchers|CreateEvent :!event Custom|PlayLandAnyway) (Variables|Default|GetFallingCheckTick) true))
  (Variables|Default|SetFallingCheckHandle _returnvalue)
  (Utilities|IsServer
    (:true
      (Animation|MulticastLinkAnimLayer (Ability|GetActivatingPawn) "/HighTower/Animation/HoneyDew/HighTower_HoneyDew_LayerAnimBP.HighTower_HoneyDew_LayerAnimBP_C"))))

(event Custom|PlayLandAnyway
  ;[1 1 1] Play land animation and remove infinite jump cd to start smash CD
  (if (not (AI|Components|NavMovement|IsFalling (Variables|/Script/Engine.Character|GetCharacterMovement (Ability|GetActivatingPawn))))
    (CallFunction|OnLand (Collision|MakeHitResult false))
    (Ability|RemoveGameplayEffectFromOwnerWithAssetTags (Variables|Default|GetInfinteJumpTag))
    (Utilities|Time|ClearandInvalidateTimerbyHandle (Variables|Default|GetFallingCheckHandle))))

(event Ability|EventOnEndAbility
  (params (bWasCancelled :type Boolean))
  ;[1 1 1] Clean up timer on end
  (|CallParentFunction bWasCancelled)
  (Utilities|Time|ClearandInvalidateTimerbyHandle (Variables|Default|GetFallingCheckHandle))
  (Utilities|IsServer
    (:true
      (Animation|MulticastUnlinkAnimLayer (Ability|GetActivatingPawn) "/HighTower/Animation/HoneyDew/HighTower_HoneyDew_LayerAnimBP.HighTower_HoneyDew_LayerAnimBP_C"))))

(event EventOnLand
  (params (Hit :type Hit%20Result%20Structure%20%28by%20ref%29))
  ;[1 1 1] Failsafe delay in cases where player gets into a bad state
  (|CallParentFunction Hit)
  (Utilities|FlowControl|Delay (Variables|Default|GetLandingFailSafeDelay))
  (CallFunction|SetBlockMovement (Variables|Default|GetPlayerPawn)))

(fn GetLaunchVelocity
  (bind (_returnvalue_roll _returnvalue_pitch _returnvalue_yaw) (Math|Rotator|BreakRotator (Pawn|GetBaseAimRotation (Variables|Default|GetPlayerPawn))))
  ;[1 1 1] Set the launch velocity to be the angle the player is looking mult by the speed. Clamping the angle to some extent.
  (return (* (Math|Rotator|GetRotationXVector (Math|Rotator|MakeRotator _returnvalue_roll (Math|Float|ClampAngle _returnvalue_pitch (ScalableFloat|GetValueatLevel (Variables|Default|GetRow_MinAngle)) (ScalableFloat|GetValueatLevel (Variables|Default|GetRow_MaxAngle))) _returnvalue_yaw)) (ScalableFloat|GetValueatLevel (Variables|Default|GetRow_LaunchSpeed)))))

(fn LandGCParams
  (bind (_normalizedmagnitude _rawmagnitude _effectcontext _matchedtagname _originaltag) (Ability|GameplayCue|BreakGameplayCueParameters (|CallParentFunction)))
  (return (Ability|GameplayCue|MakeGameplayCueParameters (Math|Conversions|ToFloat(Boolean) (Variables|PowerLeap|GetLandedInWater)) _rawmagnitude _effectcontext _matchedtagname _originaltag)))
